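# Niagara effect for the TouchLine emote: a glowing blue star with flash, light and small spinning sparkles.
# NE_Star01 layers sprite and mesh renderers through spawn-group visibility tags, growing, drifting and fading.
# NE_Star02 trails sparkles from Star01 via particle reads; NE_Star03 adds an opening sparkle burst.
# NE_Glow01/02 supply light and brief flash, with front-end modifiers adjusting offsets, velocity and flash alpha versus gameplay.
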NiagaraSystem /BRCosmetics/Animation/Game/MainPlayer/Emotes/TouchLine/FX/NS_Emote_TouchLine_Star
dependencies {
 usedDataInterfaces [
  &NiagaraDataInterfaceCurve
  &NiagaraDataInterfaceVector2DCurve
  &NiagaraDataInterfaceParticleRead
  &NiagaraDataInterfaceArrayColor
  &NiagaraDataInterfaceArrayFloat
  &NiagaraDataInterfaceColorCurve
 ]
 usedDynamicInputs []
 usedModules [&SystemState]
}
emitters {
 NE_Glow01 {
  emitterUpdateScript [
   &EmitterState(Life Cycle Mode=::NewEnumerator1, Loop Behavior=::NewEnumerator1, Loop Duration=6, UseLoopDelay=false)
   &SpawnBurst_Instantaneous(Use Spawn Probability=false, UseLoopCountLimit=false)
  ]
  particleSpawnScript [
   /SetVariables(Particles.VisibilityTag=&SpawnGroup()) {isSetParameters=true, module=SetVariables_#1}
   &InitializeParticle(
    Color=<rgba 0.0938 0.4315 2 0.2>
    Lifetime=0.7
    Mass Mode=::NewEnumerator3
    Position Offset=&FN_FrontEndVector3Modifier(Gameplay=<0 3 -5>, Front End=<0 3 -8>)
    Sprite Rotation Mode=::NewEnumerator0
    Sprite Size Mode=::NewEnumerator3
    Uniform Sprite Size=75
    UsePositionOffset=true
    Write Material Random=true
   )
  ]
  particleUpdateScript [
   &ParticleState()
   &ScaleSpriteSize(Uniform Curve Sprite Scale=DataInterface() {propertyValues {Curve {defaultValue=0, keys [{tangentMode=RCTM_User, time=0, value=0}, <at 0.02 1>, {tangentMode=RCTM_User, time=1, value=1}]}}}) {module=ScaleSpriteSize001}
   &ScaleColor(
    Scale Alpha=&FloatFromCurve(FloatCurve=DataInterface() {propertyValues {Curve {keys [<at 0.0021 0>,<at 0.0352 1>,<at 0.497 1>,<at 1 0>]}}})
    ScaleA=true
    ScaleRGB=true
   )
   &Drag(Drag=0.8, Use Linear Drag=true, Use Rotational Drag=false)
   &SpriteFacingAndAlignment(Sprite Facing=<0 1 0>, WriteAlignment=false, WriteFacing=true)
   &SolveForcesAndVelocity(Clamp Velocity=false, Limit Acceleration=false)
   &FN_Light_Attributes(
    Color=&ColorFromCurve(ColorCurve=DataInterface() {propertyValues {AlphaCurve {keys []}, BlueCurve {keys [<at 0.098 5>]}, GreenCurve {keys [<at 0.098 1.9051>]}, RedCurve {keys [<at 0.098 0>]}}})
    Create Color Attribute=true
    Create Exponent Attribute=true
    Create Position Attribute=true
    Create Radius Attribute=true
    Exponent:NiagaraFloat=&MakeCustomFloatFromBool(Bool=&Boolean_OrOperation(Bool1=$User.bIsFrontEnd, Bool2=$User.bIsFrontEndPreview), True Float=2, False Float=2)
    Position=&FN_FrontEndVector3Modifier(Gameplay=<0 3 0>, Front End=<0 5 -3>)
    Radius=&FloatFromCurve(
     FloatCurve=DataInterface() {propertyValues {Curve {defaultValue=0, keys [{tangentMode=RCTM_User, time=0, value=1}, <at 0.2532 1>, {tangentMode=RCTM_User, time=1, value=0}]}}}
     Scale Curve=&MakeCustomFloatFromBool(Bool=&Boolean_OrOperation(Bool1=$User.bIsFrontEnd, Bool2=$User.bIsFrontEndPreview), True Float=60, False Float=60)
    )
    Scale Color=&FloatFromCurve(
     FloatCurve=DataInterface() {propertyValues {Curve {defaultValue=1, keys [{tangentMode=RCTM_User, time=0, value=1}, {tangentMode=RCTM_User, time=1, value=0}]}}}
     Scale Curve=&Multiply_Float(
      A=&MakeCustomFloatFromBool(Bool=&Boolean_OrOperation(Bool1=$User.bIsFrontEnd, Bool2=$User.bIsFrontEndPreview), True Float=1, False Float=1)
      B=10
     )
    )
   )
  ]
  rendererClasses [&NiagaraLightRendererProperties, &NiagaraSpriteRendererProperties]
  renderers [
   {rendererClass=&NiagaraLightRendererProperties, rendererIndex=0}
   {rendererClass=&NiagaraSpriteRendererProperties, rendererIndex=1}
  ]
 }
 NE_Glow02 {
  emitterUpdateScript [
   &EmitterState(Life Cycle Mode=::NewEnumerator1, Loop Behavior=::NewEnumerator1, Loop Duration=6, UseLoopDelay=false)
   &SpawnBurst_Instantaneous(Use Spawn Probability=false, UseLoopCountLimit=false)
  ]
  particleSpawnScript [
   /SetVariables(Particles.Dark_Color:LinearColor=<rgba 0 0.0116 0.1615 1>) {isSetParameters=true}
   &InitializeParticle(
    Color=<rgba 0.2344 1.0788 5 0.4478>
    Lifetime=0.18
    Mass Mode=::NewEnumerator3
    Position Offset=&FN_FrontEndVector3Modifier(Gameplay=<0 5 -2>, Front End=<0 5 -5>)
    Sprite Rotation Mode=::NewEnumerator0
    Sprite Size Mode=::NewEnumerator3
    Uniform Sprite Size=380
    UsePositionOffset=true
    Write Material Random=true
   )
  ]
  particleUpdateScript [
   &ParticleState()
   &ScaleSpriteSize(Uniform Curve Sprite Scale=DataInterface() {propertyValues {Curve {defaultValue=0, keys [{tangentMode=RCTM_User, time=0, value=0}, <at 0.0556 1>, <at 0.4991 0.2>, {tangentMode=RCTM_User, time=1, value=0}]}}}) {module=ScaleSpriteSize001}
   &ScaleColor(
    Scale Alpha=&FloatFromCurve(
     FloatCurve=DataInterface() {propertyValues {Curve {keys [<at 0.0021 0>,<at 0.0352 1>,<at 0.497 1>,<at 1 0>]}}}
     Scale Curve=&FN_FrontEndFloatModifier(Gameplay=0.5, Front End=0.15)
    )
    ScaleA=true
    ScaleRGB=true
   )
   /SetVariables(Particles.Dark_Color:LinearColor=&MakeLinearColorFromVectorAndFloat(
     "Vector (RGB)"=&MakeVectorFromLinearColorRGB(Color=$Particles.Initial.Dark_Color)
     "Float (Alpha)"=&FloatFromCurve(FloatCurve=DataInterface() {propertyValues {Curve {keys [<at 0.0021 0>,<at 0.0352 1>,<at 0.497 1>,<at 1 0>]}}}, Scale Curve=0.6)
    )) {isSetParameters=true}
   &SpriteFacingAndAlignment(Sprite Facing=<0 1 0>, WriteAlignment=false, WriteFacing=true)
   /SetVariables(Particles.Dark_SpriteSize:Vector2f=&Multiply_Vector2D_ByFloat(Float=1.1, Vector2D=$Particles.SpriteSize)) {isSetParameters=true}
   &SolveForcesAndVelocity(Clamp Velocity=false, Limit Acceleration=false)
  ]
  rendererClasses [&NiagaraSpriteRendererProperties]
  renderers [
   {rendererClass=&NiagaraSpriteRendererProperties}
   {rendererClass=&NiagaraSpriteRendererProperties, rendererIndex=1}
  ]
 }
 NE_Star01 {
  emitterUpdateScript [
   &EmitterState(Life Cycle Mode=::NewEnumerator1, Loop Behavior=::NewEnumerator1, Loop Duration=6, UseLoopDelay=false)
   &SpawnBurst_Instantaneous(Spawn Time=0.05, Use Spawn Probability=false, UseLoopCountLimit=false)
   &SpawnBurst_Instantaneous(Spawn Group=1, Spawn Time=0.05, Use Spawn Probability=false, UseLoopCountLimit=false) {module=SpawnBurst_Instantaneous001}
   &SpawnBurst_Instantaneous(Spawn Group=2, Spawn Time=0.05, Use Spawn Probability=false, UseLoopCountLimit=false) {module=SpawnBurst_Instantaneous002}
   &SpawnBurst_Instantaneous(Spawn Group=3, Spawn Time=0.05, Use Spawn Probability=false, UseLoopCountLimit=false) {module=SpawnBurst_Instantaneous003}
   &SpawnBurst_Instantaneous(Spawn Group=4, Spawn Time=0.05, Use Spawn Probability=false, UseLoopCountLimit=false) {module=SpawnBurst_Instantaneous004}
   &SpawnBurst_Instantaneous(Spawn Group=5, Spawn Time=0.05, Use Spawn Probability=false, UseLoopCountLimit=false) {module=SpawnBurst_Instantaneous005}
   &SpawnBurst_Instantaneous(Spawn Group=6, Spawn Time=0.05, Use Spawn Probability=false, UseLoopCountLimit=false) {module=SpawnBurst_Instantaneous006}
   &SpawnBurst_Instantaneous(Spawn Group=7, Spawn Time=0.05, Use Spawn Probability=false, UseLoopCountLimit=false) {module=SpawnBurst_Instantaneous007}
  ]
  particleSpawnScript [
   /SetVariables(Particles.VisibilityTag=&SpawnGroup()) {isSetParameters=true, module=SetVariables_#1}
   &InitializeParticle(
    Color=&MergeHelper_ColorSwitch(
     0=<rgba 0.0044 0 1 0.9043>
     1=<rgba 0.0625 0.5399 2 0.3696>
     2=<rgba 0.2891 0.825 3 0.4696>
     3=<rgba 0 0.0179 0.1458 0.4696>
     4=<rgba 0.195 0.7356 5 0.1304>
     5=<rgba 0 0.0179 0.1458 0.4696>
     6=<rgba 0.0625 0.5399 2 0.4696>
     7=<rgba 0.0625 0.5399 2 0.2174>
     Index=&SpawnGroup()
     Num=::NewEnumerator6
    )
    Lifetime=4.5
    Mass Mode=::NewEnumerator3
    Mesh Scale Mode=::NewEnumerator3
    Mesh Uniform Scale=0.64
    Position Offset=&MergeHelper_VectorSwitch(
     0=<0 0 0>
     1=<0 0 0>
     2=<0 0 0>
     3=<0 0 0>
     4=<0 0 -1.5>
     5=<0 0 0>
     6=<0 0 0>
     7=<0 0 0>
     Index=&SpawnGroup()
     Num=::NewEnumerator6
    )
    Sprite Rotation Mode=::NewEnumerator0
    Sprite Size=&MergeHelper_Vector2Switch(
     0=<120 120>
     1=<210 210>
     2=<135 135>
     3=<300 300>
     4=<210 210>
     5=<420 420>
     6=<200 30>
     7=<210 210>
     Index=&SpawnGroup()
     Num=::NewEnumerator6
    )
    Sprite Size Mode=::NewEnumerator4
    UsePositionOffset=true
    Write Material Random=true
   )
   &InitialMeshOrientation(Orientation Vector=<0 0 1>, Rotation=<0.25 0 0>, Use Orientation Vector=true, Use Rotation Vector=true)
   &InitialMeshOrientation(Orientation Axis=<0 0 0>, Orientation Vector=<0 0 0>, Rotation=<0 0.14 0>, Use Orientation Vector=true, Use Rotation Vector=true) {module=InitialMeshOrientation001}
   &AddVelocity(Velocity=&FN_FrontEndVector3Modifier(Gameplay=<0 20 3>, Front End=<0 20 2>))
   /SetVariables(Particles.Position=&AddVectorToPosition(Position=<0 0 0>, Vector=&FN_FrontEndVector3Modifier(Gameplay=<0 0 0>, Front End=<0 0 -3>))) {isSetParameters=true}
  ]
  particleUpdateScript [
   &ParticleState()
   &ScaleMeshSize(Scale Factor=&MakeVector(
     X=&FloatFromCurve(
      FloatCurve=DataInterface() {propertyValues {Curve {
         defaultValue=0
         keys [
          {arriveTangent=2.6685, interpMode=RCIM_Cubic, leaveTangent=2.6685, tangentMode=RCTM_User, time=0, value=0.25}
          {arriveTangent=0.0224, interpMode=RCIM_Cubic, leaveTangent=0.0224, tangentMode=RCTM_User, time=1, value=1.25}
         ]
        }}}
      Scale Curve=&Multiply_Float(
       A=1
       B=&Waveform(
        "Waveform [1]"=::NewEnumerator0
        "[1] Phase Scale"=2
        "[1] Frequency"=1
        "[1] Amplitude Min/Max"=&MakeVector2D(X=&RandomRangeFloat(Minimum=0.95, Maximum=0.98), Y=1)
       )
      )
     )
     Y=&FloatFromCurve(
      FloatCurve=DataInterface() {propertyValues {Curve {
         defaultValue=0
         keys [
          {arriveTangent=2.6685, interpMode=RCIM_Cubic, leaveTangent=2.6685, tangentMode=RCTM_User, time=0, value=0.25}
          {arriveTangent=0.0224, interpMode=RCIM_Cubic, leaveTangent=0.0224, tangentMode=RCTM_User, time=1, value=1.25}
         ]
        }}}
      Scale Curve=&Multiply_Float(A=1, B=1)
     )
     Z=&FloatFromCurve(
      FloatCurve=DataInterface() {propertyValues {Curve {
         defaultValue=1
         keys [
          {arriveTangent=10.0278, time=0, value=0.1}
          {arriveTangent=2.3122, time=0.03, value=1}
          {arriveTangent=1.6613, leaveTangent=1.6613, time=1.0048, value=2}
         ]
        }}}
      Scale Curve=&Multiply_Float(A=1, B=1)
     )
    ))
   /SetVariables(Particles.Glow_SpriteSize:Vector2f=&Multiply_Vector2D_ByFloat(
     Float=&FloatFromCurve(FloatCurve=DataInterface() {propertyValues {Curve {
         defaultValue=0
         keys [
          {arriveTangent=2.6685, interpMode=RCIM_Cubic, leaveTangent=2.6685, tangentMode=RCTM_User, time=0, value=0.25}
          {arriveTangent=0.0224, interpMode=RCIM_Cubic, leaveTangent=0.0224, tangentMode=RCTM_User, time=1, value=1.25}
         ]
        }}})
     Vector2D=$Particles.Initial.SpriteSize
    )) {isSetParameters=true}
   &ScaleSpriteSize(
    Scale Factor=&MakeVector2D(
     X=&FloatFromCurve(
      FloatCurve=DataInterface() {propertyValues {Curve {
         defaultValue=0
         keys [
          {arriveTangent=2.6685, interpMode=RCIM_Cubic, leaveTangent=2.6685, tangentMode=RCTM_User, time=0, value=0.25}
          {arriveTangent=0.0224, interpMode=RCIM_Cubic, leaveTangent=0.0224, tangentMode=RCTM_User, time=1, value=1.25}
         ]
        }}}
      Scale Curve=&Multiply_Float(
       A=1
       B=&Waveform(
        "Waveform [1]"=::NewEnumerator0
        "[1] Phase Scale"=2
        "[1] Frequency"=1
        "[1] Amplitude Min/Max"=&MakeVector2D(X=&RandomRangeFloat(Minimum=0.95, Maximum=0.98), Y=1)
       )
      )
     )
     Y=&FloatFromCurve(
      FloatCurve=DataInterface() {propertyValues {Curve {
         defaultValue=0
         keys [
          {arriveTangent=2.6685, interpMode=RCIM_Cubic, leaveTangent=2.6685, tangentMode=RCTM_User, time=0, value=0.25}
          {arriveTangent=0.0224, interpMode=RCIM_Cubic, leaveTangent=0.0224, tangentMode=RCTM_User, time=1, value=1.25}
         ]
        }}}
      Scale Curve=&Multiply_Float(A=1, B=1)
     )
    )
    Scale Sprite Size Mode=::NewEnumerator2
   ) {module=ScaleSpriteSize001}
   &ScaleColor(
    Scale Alpha=&MergeHelper_FloatSwitch(
     0=&FloatFromCurve(FloatCurve=DataInterface() {propertyValues {Curve {keys [<at 0.2 1>, {arriveTangent=-1.0036, time=0.6047, value=1}, <at 1 0 RCIM_Cubic>]}}})
     1=&FloatFromCurve(FloatCurve=DataInterface() {propertyValues {Curve {keys [<at 0.2 1>, {arriveTangent=-1.0036, time=0.6047, value=1}, <at 1 0 RCIM_Cubic>]}}})
     2=&FloatFromCurve(FloatCurve=DataInterface() {propertyValues {Curve {keys [{arriveTangent=-1.0036, time=0.4, value=1}, <at 0.93 0 RCIM_Cubic>]}}})
     3=&FloatFromCurve(FloatCurve=DataInterface() {propertyValues {Curve {keys [<at 0.2 1>, {arriveTangent=-1.0036, time=0.6047, value=1}, <at 1 0 RCIM_Cubic>]}}})
     4=&FloatFromCurve(FloatCurve=DataInterface() {propertyValues {Curve {keys [<at 0.2 1>, {arriveTangent=-1.0036, time=0.5456, value=1}, <at 0.95 0 RCIM_Cubic>]}}})
     5=&FloatFromCurve(FloatCurve=DataInterface() {propertyValues {Curve {keys [<at 0.2 1>, {arriveTangent=-1.0036, time=0.6047, value=1}, <at 1 0 RCIM_Cubic>]}}})
     6=&FloatFromCurve(FloatCurve=DataInterface() {propertyValues {Curve {keys [<at 0.2 1>, {arriveTangent=-1.0036, time=0.6047, value=1}, <at 1 0 RCIM_Cubic>]}}})
     7=&FloatFromCurve(FloatCurve=DataInterface() {propertyValues {Curve {keys [<at 0.2 1>, {arriveTangent=-1.0036, time=0.6047, value=1}, <at 1 0 RCIM_Cubic>]}}})
     Index=&SpawnGroup()
     Num=::NewEnumerator6
    )
    ScaleA=true
    ScaleRGB=true
   )
   &Drag(Drag=0.5, Use Linear Drag=true, Use Rotational Drag=false)
   &SpriteFacingAndAlignment(Sprite Facing=<0 1 0>, WriteAlignment=true, WriteFacing=true)
   /SetVariables() {enabled=false, isSetParameters=true}
   &SolveForcesAndVelocity(Clamp Velocity=false, Limit Acceleration=false)
  ]
  rendererClasses [&NiagaraSpriteRendererProperties, &NiagaraMeshRendererProperties]
  renderers [
   {rendererClass=&NiagaraSpriteRendererProperties}
   {rendererClass=&NiagaraSpriteRendererProperties, rendererIndex=1}
   {rendererClass=&NiagaraSpriteRendererProperties, rendererIndex=2}
   {rendererClass=&NiagaraSpriteRendererProperties, rendererIndex=3}
   {rendererClass=&NiagaraMeshRendererProperties, rendererIndex=4}
   {rendererClass=&NiagaraSpriteRendererProperties, rendererIndex=5}
   {rendererClass=&NiagaraSpriteRendererProperties, rendererIndex=6}
   {rendererClass=&NiagaraSpriteRendererProperties, rendererIndex=7}
  ]
 }
 NE_Star02 {
  emitterUpdateScript [
   &EmitterState(Life Cycle Mode=::NewEnumerator1, Loop Behavior=::NewEnumerator1, Loop Duration=3, UseLoopDelay=false)
   &SpawnParticlesFromOtherEmitter(
    Attribute Reader=DataInterface() {propertyValues {EmitterBinding {bindingMode=Other, emitterName=NE_Star01}}}
    Calculate Spawn Rate Per Particle=false
    Spawn Rate=20
    Use Maximum Age=false
    Use Minimum Age=false
   )
  ]
  particleSpawnScript [
   &InitializeParticle(
    Color=&SelectLinearColorFromArray(
     Array Sampling Mode=::NewEnumerator0
     Direct Array Index=?
     Color Selection Array=DataInterface() {propertyValues {ColorData [<rgba 0.073 0.3642 1 1>, <rgba 0 0.1243 0.7005 1>]}}
     Random Seed=0
    )
    Lifetime=&RandomRangeFloat(Minimum=1, Maximum=2)
    Sprite Size Mode=::NewEnumerator1
    Uniform Sprite Size Max=15
    UsePositionOffset=false
    Write Material Random=true
   )
   /SetVariables(Particles.GlowColor=<rgba 0.0328 0.3714 1.2 0.9>) {isSetParameters=true, module=SetVariables_#2}
   &ShapeLocation(
    Cone Angle=80
    Cone Inner Angle=70
    Cone Length=0.1
    MaskSpawnGroup=false
    Offset Mode=::NewEnumerator0
    Override Seed=false
    Rotation Mode=::NewEnumerator2
    Shape Primitive=::NewEnumerator5
    UseOwnerScale=true
    Yaw / Pitch / Roll=<0 0 0.25>
   ) {module=ShapeLocation003}
   &AddVelocity(
    Module.Use Falloff Distance=false
    Override Seed=false
    Velocity Mode=::NewEnumerator1
    Velocity Speed=&RandomRangeFloat(Minimum=120, Maximum=250)
   )
   &SampleParticlesFromOtherEmitter(Applied Position Offset:Vector3f=<0 -15 0>)
  ]
  particleUpdateScript [
   &ParticleState()
   &SpriteRotationRate(Rotation Rate=&SelectFloatFromArray(
     Array Sampling Mode=::NewEnumerator0
     Direct Array Index=?
     Float Selection Array=DataInterface() {propertyValues {FloatData [-100,100]}}
     Random Seed=0
    ))
   &Drag(Drag=1.8, Use Linear Drag=true, Use Rotational Drag=false)
   /SetVariables(Particles.GlowColor=&MakeLinearColorFromVectorAndFloat(
     "Vector (RGB)"=&MakeVectorFromLinearColorRGB(Color=$Particles.Initial.GlowColor)
     "Float (Alpha)"=&Multiply_Float(
      A=0.8
      B=&FloatFromCurve(FloatCurve=DataInterface() {propertyValues {Curve {keys [<at -0.0092 0.0144>,<at 0.0241 1>,<at 0.45 1>,<at 1.0011 0.002>]}}})
     )
    )) {isSetParameters=true, module=SetVariables_#3}
   &ScaleColor(
    Scale Alpha=&FloatFromCurve(FloatCurve=DataInterface() {propertyValues {Curve {keys [<at -0.0092 0.0144>,<at 0.0967 0.9767>,<at 0.6023 1>,<at 1.0011 0.002>]}}})
    ScaleA=true
    ScaleRGB=true
   ) {module=ScaleColor001}
   &ScaleSpriteSize(Uniform Curve Sprite Scale=DataInterface() {propertyValues {Curve {defaultValue=0.0716, keys [{tangentMode=RCTM_User, time=0, value=0}, <at 0.1042 0.85>, {tangentMode=RCTM_User, time=1, value=1}]}}})
   /SetVariables(Particles.Glow_SpriteSize:Vector2f=&Multiply_Vector2D_ByFloat(Float=1.5, Vector2D=$Particles.SpriteSize)) {isSetParameters=true, module=SetVariables_#4}
   &SolveForcesAndVelocity(Clamp Velocity=false, Limit Acceleration=false)
  ]
  rendererClasses [&NiagaraSpriteRendererProperties]
  renderers [
   {rendererClass=&NiagaraSpriteRendererProperties}
   {rendererClass=&NiagaraSpriteRendererProperties, rendererIndex=1}
  ]
 }
 NE_Star03 {
  emitterUpdateScript [
   &EmitterState(Life Cycle Mode=::NewEnumerator1, Loop Behavior=::NewEnumerator1, Loop Duration=3, UseLoopDelay=false)
   &SpawnBurst_Instantaneous(Spawn Count=5, Use Spawn Probability=false, UseLoopCountLimit=false)
  ]
  particleSpawnScript [
   &InitializeParticle(
    Color=&SelectLinearColorFromArray(
     Array Sampling Mode=::NewEnumerator0
     Direct Array Index=?
     Color Selection Array=DataInterface() {propertyValues {ColorData [<rgba 0.073 0.3642 1 1>, <rgba 0 0.1243 0.7005 1>]}}
     Random Seed=0
    )
    Lifetime=&RandomRangeFloat(Minimum=1, Maximum=2)
    Sprite Size Mode=::NewEnumerator1
    Uniform Sprite Size Max=15
    UsePositionOffset=false
    Write Material Random=true
   )
   /SetVariables(Particles.GlowColor=<rgba 0.0328 0.3714 1.2 0.9>) {isSetParameters=true, module=SetVariables_#2}
   &ShapeLocation(
    Cone Angle=80
    Cone Inner Angle=70
    Cone Length=0.1
    MaskSpawnGroup=false
    Offset Mode=::NewEnumerator0
    Override Seed=false
    Rotation Mode=::NewEnumerator2
    Shape Primitive=::NewEnumerator5
    UseOwnerScale=true
    Yaw / Pitch / Roll=<0 0 0.25>
   ) {module=ShapeLocation003}
   &AddVelocity(
    Module.Use Falloff Distance=false
    Override Seed=false
    Velocity Mode=::NewEnumerator1
    Velocity Speed=&RandomRangeFloat(Minimum=120, Maximum=250)
   )
  ]
  particleUpdateScript [
   &ParticleState()
   &SpriteRotationRate(Rotation Rate=&SelectFloatFromArray(
     Array Sampling Mode=::NewEnumerator0
     Direct Array Index=?
     Float Selection Array=DataInterface() {propertyValues {FloatData [-100,100]}}
     Random Seed=0
    ))
   &Drag(Drag=2.3, Use Linear Drag=true, Use Rotational Drag=false)
   /SetVariables(Particles.GlowColor=&MakeLinearColorFromVectorAndFloat(
     "Vector (RGB)"=&MakeVectorFromLinearColorRGB(Color=$Particles.Initial.GlowColor)
     "Float (Alpha)"=&Multiply_Float(
      A=0.8
      B=&FloatFromCurve(FloatCurve=DataInterface() {propertyValues {Curve {keys [<at -0.0092 0.0144>,<at 0.0241 1>,<at 0.45 1>,<at 1.0011 0.002>]}}})
     )
    )) {isSetParameters=true, module=SetVariables_#3}
   &ScaleColor(
    Scale Alpha=&FloatFromCurve(FloatCurve=DataInterface() {propertyValues {Curve {keys [<at -0.0092 0.0144>,<at 0.0967 0.9767>,<at 0.6023 1>,<at 1.0011 0.002>]}}})
    ScaleA=true
    ScaleRGB=true
   ) {module=ScaleColor001}
   &ScaleSpriteSize(Uniform Curve Sprite Scale=DataInterface() {propertyValues {Curve {defaultValue=0.0716, keys [{tangentMode=RCTM_User, time=0, value=0}, <at 0.1042 0.85>, {tangentMode=RCTM_User, time=1, value=1}]}}})
   /SetVariables(Particles.Glow_SpriteSize:Vector2f=&Multiply_Vector2D_ByFloat(Float=1.5, Vector2D=$Particles.SpriteSize)) {isSetParameters=true, module=SetVariables_#4}
   &SolveForcesAndVelocity(Clamp Velocity=false, Limit Acceleration=false)
  ]
  rendererClasses [&NiagaraSpriteRendererProperties]
  renderers [
   {rendererClass=&NiagaraSpriteRendererProperties}
   {rendererClass=&NiagaraSpriteRendererProperties, rendererIndex=1}
  ]
 }
}
summary {
 systemName=NS_Emote_TouchLine_Star
 emitters [
  {emitterName=NE_Star01, rendererClasses [&NiagaraSpriteRendererProperties, &NiagaraMeshRendererProperties]}
  {emitterName=NE_Star02, rendererClasses [&NiagaraSpriteRendererProperties]}
  {emitterName=NE_Glow01, rendererClasses [&NiagaraLightRendererProperties, &NiagaraSpriteRendererProperties]}
  {emitterName=NE_Glow02, rendererClasses [&NiagaraSpriteRendererProperties]}
  {emitterName=NE_Star03, rendererClasses [&NiagaraSpriteRendererProperties]}
 ]
 userVariables [
  User.AngleFade {type=NiagaraFloat, defaultValue=5, description=""}
  User.bIsFrontEnd {type=NiagaraBool}
  User.bIsFrontEndPreview {type=NiagaraBool}
 ]
}
